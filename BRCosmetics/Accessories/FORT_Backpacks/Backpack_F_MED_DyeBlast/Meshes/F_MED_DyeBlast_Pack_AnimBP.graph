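# Animation Blueprint for the DyeBlast backpack cosmetic, adding a pose offset and secondary motion to the pack.
# The input pose passes through two additive bone-space Modify Bone nodes on the pack bone, the second with a bool alpha.
# Chained Anim Dynamics nodes then simulate the handle and chain bones root-relative, with angular springs, angle limits and reduced gravity.
# The event graph holds only unconnected default nodes, and no variables are defined.
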AnimBlueprint /BRCosmetics/Accessories/FORT_Backpacks/Backpack_F_MED_DyeBlast/Meshes/F_MED_DyeBlast_Pack_AnimBP
target_skeleton=&/F_MED_DyeBlast_Pack_Skeleton
graphs [
 AnimGraph {nodes [
   n1 = "Transform (Modify) Bone - Bone: pack"(ComponentPose=Local To Component(LocalPose=Input Pose()), Translation="0,2,5", Rotation="0,0,5") {
    rotationMode=BMM_Additive
    rotationSpace=BCS_BoneSpace
    translationMode=BMM_Additive
    translationSpace=BCS_BoneSpace
    boneToModify {boneName=pack}
    rotation {roll=5}
    translation {y=2, z=5}
   }
   n2 = "Transform (Modify) Bone - Bone: pack"!(ComponentPose=@n1, bAlphaBoolEnabled:bool, Translation="0,0,0", Rotation="0,0,5", Scale) -> (Pose) {
    alphaInputType=Bool
    rotationMode=BMM_Additive
    rotationSpace=BCS_BoneSpace
    translationMode=BMM_Additive
    translationSpace=BCS_BoneSpace
    boneToModify {boneName=pack}
    rotation {roll=5}
   }
   n3 = Anim Dynamics - Bone: dyn_handle(ComponentPose=@n2, Alpha=1, ExternalForce="0,0,0") {
    angularDampingOverride=1
    angularSpringConstant=20000
    bAngularSpring=true
    bOverrideAngularDamping=true
    bUseGravityOverride=true
    gravityScale=0.35
    simulationSpace=RootRelative
    boundBone {boneName=dyn_handle}
    componentLinearAccScale=<1 1 1>
    componentLinearVelScale=<1 1 1>
    gravityOverride {z=-980}
    physicsBodyDefinitions [{
      boundBone {boneName=dyn_handle}
      boxExtents=<15 25 25>
      constraintSetup {angularLimitsMax=<0 20 0>, angularLimitsMin=<0 -20 0>, angularTarget=<1 0 0>}
      localJointOffset=<2 0 0>
     }]
   }
   n4 = Anim Dynamics - Bone: dyn_chain(ComponentPose=@n3, Alpha=1, ExternalForce="0,0,0") {
    angularDampingOverride=1
    angularSpringConstant=5000
    bAngularSpring=true
    bOverrideAngularDamping=true
    bUseGravityOverride=true
    gravityScale=0.35
    simulationSpace=RootRelative
    boundBone {boneName=dyn_chain}
    componentLinearAccScale=<1 1 1.5>
    componentLinearVelScale=<1 1 1.5>
    gravityOverride {z=-800}
    physicsBodyDefinitions [{
      boundBone {boneName=dyn_chain}
      boxExtents=<15 20 20>
      constraintSetup {angularLimitsMax=<0 0 20>, angularLimitsMin=<0 -50 -20>, angularTarget=<1 0 0>}
      localJointOffset=<3 0 0>
     }]
   }
   Output Pose(Result=Component To Local(ComponentPose=@n4)) {layerGroup=None}
  ]}
 EventGraph {nodes [
   TryGetPawnOwner()
   Event BlueprintUpdateAnimation()
  ]}
]
variables []
